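# UI-domain material for the Winterfest 2022 Item Shop background, building an animated aurora night sky.
# A blue radial gradient blends through an aurora mask into a panning, flow-mapped wave texture tinted teal, then darkens under a rotated top-bottom gradient.
# A star layer multiplies sky-dome star masks, one panning for flicker, with StarFlickerSpeed and StarPower parameters.
# The graph holds two near-duplicate versions of the color chain with differing tints and gradient settings.
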Material /ItemShop/UI/Art/Winterfest/WF_2022/M_UI_WF22_StoreBG
graph {
 properties {bUsedWithStaticMesh=false, materialDomain=MD_UI, pixelDepthOffsetMode=PDOM_Legacy, usedShadingModels=Default Lit}
 nodes [
  n1 = Constant3Vector() {constant=<rgba 0.0097 0.1016 0.4062 1>}
  n2 = MaterialFunctionCall(
   UV=MaterialFunctionCall(UVs=NamedRerouteUsage(), Texture Scale=Constant() {r=1.588}) {materialFunction=&ScaleUVsByCenter}.UVs
   GradientStart=Constant() {r=-0.1627}
   GradientEnd=Constant() {r=1.0115}
  ) {materialFunction=&MF_UI_RadialGradient}
  n3 = TextureObject() {samplerType=SAMPLERTYPE_LinearGrayscale, texture=/Game/UI/Frontend/ItemShop/Art/Winterfest/WF_2022/T_UI_WaveTexture}
  n4 = NamedRerouteUsage()
  n5 = Panner(Coordinate=Multiply(A=@n4) {constB=0.5}) {speedY=0.05}
  n6 = TextureObject() {samplerType=SAMPLERTYPE_LinearColor, texture=/Game/UI/Frontend/ItemShop/Art/WaterBackground/T_UI_WaterFlowMap}
  n7 = Constant() {r=0.3}
  n8 = Constant() {r=0.01}
  n9 = NamedRerouteUsage()
  n10 = LinearInterpolate(
   A=LinearInterpolate(A=Constant3Vector() {constant=<rgba 0.0016 0.0062 0.0595 1>}, B=@n1, Alpha=Power(Base=@n2.Smooth))
   B=LinearInterpolate(
    A=@n1
    B=Constant3Vector() {constant=<rgba 0.0518 0.4323 0.3514 0>}
    Alpha=MaterialFunctionCall!(Texture 2D=@n3, Texture UVs=@n5, FlowMap Texture 2D=@n6, FlowMap UVs=@n4, FlowIntensity=@n7, FlowSpeed=@n8) {materialFunction=/FortUILibrary/MaterialFunctions/Effects/MF_UI_FlowMap}.Result
   )
   Alpha=Multiply(A=TextureSample(UVs=@n9) {texture=&/T_UI_Aurora}.R) {constB=1.25}
  )
  LinearInterpolate(A=@n10, B=Constant3Vector() {constant=<rgba 0.7198 0.7503 0.8177 1>}.B, Alpha=Multiply() {constB=1.1})
  n11 = MaterialFunctionCall(UVs=NamedRerouteUsage(), "Rotation Angle (0-1)"=Constant() {r=-0.45}) {materialFunction=&CustomRotator}
  n12 = Constant() {r=-0.2}
  n13 = Constant() {r=1}
  n14 = Constant() {r=1.25}
  n15 = Saturate(None=SmoothStep(Value=MaterialFunctionCall!(UVs=@n11.Rotated Values, Subtract=@n12, Falloff=@n13, Intensity=@n14) {materialFunction=/Game/UI/Frontend/ItemShop/Art/_MaterialFunctions_ShopTheming/MF_UI_TopBottomGradient}.Result) {constMax=1.25})
  n16 = Multiply(A=@n15)
  n17 = NamedRerouteUsage()
  LinearInterpolate(
   A=LinearInterpolate(A=@n10, B=Constant3Vector() {constant=<rgba 0.0026 0.0045 0.0312 1>}, Alpha=@n15)
   B=Constant3Vector() {constant=<rgba 0.7639 0.7738 0.9167 1>}
   Alpha=Min(A=OneMinus(None=@n16), B=Multiply(A=@n17) {constB=2})
  )
  n18 = Constant3Vector() {constant=<rgba 0.0176 0.1845 0.7379 1>}
  n19 = MaterialFunctionCall(UVs=NamedRerouteUsage(), Texture Scale=Constant() {r=1.588}) {materialFunction=&ScaleUVsByCenter}
  n20 = Constant() {r=-0.1627}
  n21 = Constant() {r=1.0115}
  n22 = LinearInterpolate(
   A=Constant3Vector() {constant=<rgba 0.0015 0.0062 0.0595 1>}
   B=@n18
   Alpha=Power(Base=MaterialFunctionCall(UV=@n19.UVs, GradientStart=@n20, GradientEnd=@n21) {materialFunction=&MF_UI_RadialGradient}.Smooth)
  )
  n23 = Multiply(A=NamedRerouteUsage()) {constB=0.5}
  n24 = LinearInterpolate(
   A=@n18
   B=Constant3Vector() {constant=<rgba 0 1 0.7874 0>}
   Alpha=MaterialFunctionCall(
    Texture 2D=TextureObject() {samplerType=SAMPLERTYPE_LinearGrayscale, texture=&T_UI_WaveTexture}
    Texture UVs=Panner(Coordinate=@n23) {speedY=0.05}
    FlowMap Texture 2D=TextureObject() {samplerType=SAMPLERTYPE_LinearColor, texture=&T_UI_WaterFlowMap}
    FlowIntensity=Constant() {r=0.3}
    FlowSpeed=Constant() {r=0.01}
   ) {materialFunction=&MF_UI_FlowMap}.Result
  )
  n25 = TextureSample(UVs=NamedRerouteUsage()) {texture=&T_UI_Aurora}
  n26 = Multiply(A=@n25.R) {constB=1.25}
  n27 = MaterialFunctionCall(UVs=NamedRerouteUsage(), "Rotation Angle (0-1)"=Constant() {r=-0.45}) {materialFunction=&CustomRotator}
  n28 = Constant() {r=0.2}
  n29 = Constant() {r=0.95}
  n30 = Constant() {r=1.25}
  LinearInterpolate(
   A=LinearInterpolate(A=LinearInterpolate(A=@n22, B=@n24, Alpha=@n26), B=Constant3Vector() {constant=<rgba 0.7198 0.7503 0.8177 1>}.B, Alpha=@n25.B)
   B=Constant3Vector() {constant=<rgba 0.006 0.0103 0.0723 1>}
   Alpha=Saturate(None=MaterialFunctionCall(UVs=@n27.Rotated Values, Subtract=@n28, Falloff=@n29, Intensity=@n30) {materialFunction=&MF_UI_TopBottomGradient}.Result)
  )
  n31 = Constant2Vector() {g=1024, r=2048}
  MaterialFunctionCall!(ScreenUV, TextureSize=@n31, ScalePivotPoint) {materialFunction=/FortUILibrary/MaterialFunctions/Basics/MF_UI_ScreenSpaceUVScaling}
  n32 = MaterialFunctionCall(TextureSize=@n31) {materialFunction=&MF_UI_UVScaling}
  NamedRerouteDeclaration(None=@n32.Scaled to Fit) {name=UV Scaling_1, nodeColor=<rgba 0 0.0625 1 1>}
  NamedRerouteDeclaration(None=@n32.Scaled To Fill) {name=UVScaling Fill_1, nodeColor=<rgba 0.5469 1 0 1>}
  n33 = MaterialFunctionCall() {materialFunction=&MF_UI_ScreenSpaceUVScaling}
  n34 = TextureSample(UVs=Multiply(A=@n33.Scaled to Fit) {constB=1.5}) {constMipValue=1, samplerType=SAMPLERTYPE_LinearColor, texture=/Game/Packages/Fortress_Sky/SkyDome/Master/T_SkyDome_StarCloudMask}
  n35 = TextureSample(UVs=Panner(
    Coordinate=Multiply(A=@n33.Scaled to Fit) {constB=0.1}
    Speed=ScalarParameter() {defaultValue=0.025, group=Stars, parameterName=StarFlickerSpeed}
   )) {samplerType=SAMPLERTYPE_LinearColor, texture=&T_SkyDome_StarCloudMask}
  n36 = ShadingPathSwitch(Default=Multiply(A=ComponentMask(None=@n34.R) {r=true}, B=Multiply(A=@n35.G) {constB=0.3}))
  n37 = ScalarParameter() {defaultValue=2, group=Stars, parameterName=StarPower}
  NamedRerouteDeclaration(None=Multiply(A=Power(Base=@n36, Exp=@n37), B=Constant() {r=100})) {name=Stars_1, nodeColor=<rgba 0.8414 0 1 1>}
  NamedRerouteUsage()
  NamedRerouteUsage()
 ]
}
