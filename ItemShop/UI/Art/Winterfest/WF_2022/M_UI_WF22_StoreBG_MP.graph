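# Winterfest 2022 Item Shop background material that paints a night sky with an animated aurora and flickering stars.
# An aurora texture masks flow-mapped, panning blue-to-teal waves over a radial navy gradient, which a rotated linear gradient then darkens.
# Stars multiply static and panning samples of a skydome mask in screen space, with parameters for flicker speed and power.
# The slightly desaturated composite feeds the shared MPItemShop setup function as its color.
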Material /ItemShop/UI/Art/Winterfest/WF_2022/M_UI_WF22_StoreBG_MP
graph {
 properties {pixelDepthOffsetMode=PDOM_Legacy, usedShadingModels=Default Lit}
 nodes [
  n1 = Constant3Vector() {constant=<rgba 0.0097 0.1016 0.4062 1>}
  n2 = NamedRerouteUsage()
  n3 = Multiply(A=@n2) {constB=0.5}
  n4 = LinearInterpolate(
   A=@n1
   B=Constant3Vector() {constant=<rgba 0.0518 0.4323 0.3514 0>}
   Alpha=MaterialFunctionCall!(
    Texture 2D=TextureObject() {samplerType=SAMPLERTYPE_LinearGrayscale, texture=/Game/UI/Frontend/ItemShop/Art/Winterfest/WF_2022/T_UI_WaveTexture}
    Texture UVs=Panner(Coordinate=@n3) {speedY=0.05}
    FlowMap Texture 2D=TextureObject() {samplerType=SAMPLERTYPE_LinearColor, texture=/Game/UI/Frontend/ItemShop/Art/WaterBackground/T_UI_WaterFlowMap}
    FlowMap UVs=@n2
    FlowIntensity=Constant() {r=0.3}
    FlowSpeed=Constant() {r=0.01}
   ) {materialFunction=/FortUILibrary/MaterialFunctions/Effects/MF_UI_FlowMap}.Result
  )
  n5 = Multiply(A=TextureSample(UVs=NamedRerouteUsage()) {texture=&/T_UI_Aurora}.R) {constB=1.25}
  n6 = MaterialFunctionCall(UVs=NamedRerouteUsage(), Texture Scale=Constant() {r=1.588}) {materialFunction=&ScaleUVsByCenter}
  n7 = Constant() {r=-0.1627}
  n8 = Constant() {r=1.0115}
  n9 = LinearInterpolate(
   A=Constant3Vector() {constant=<rgba 0.0016 0.0062 0.0595 1>}
   B=@n1
   Alpha=Power(Base=MaterialFunctionCall(UV=@n6.UVs, GradientStart=@n7, GradientEnd=@n8) {materialFunction=&MF_UI_RadialGradient}.Smooth)
  )
  n10 = MaterialFunctionCall(UVs=NamedRerouteUsage(), "Rotation Angle (0-1)"=Constant() {r=-0.45}) {materialFunction=&CustomRotator}
  n11 = Constant() {r=-0.2}
  n12 = Constant() {r=1}
  n13 = Constant() {r=1.25}
  n14 = Saturate(None=SmoothStep(Value=MaterialFunctionCall!(UVs=@n10.Rotated Values, Subtract=@n11, Falloff=@n12, Intensity=@n13) {materialFunction=/Game/UI/Frontend/ItemShop/Art/_MaterialFunctions_ShopTheming/MF_UI_TopBottomGradient}.Result) {constMax=1.25})
  n15 = Multiply(A=@n14)
  n16 = NamedRerouteUsage()
  n17 = LinearInterpolate(
   A=LinearInterpolate(A=LinearInterpolate(A=@n9, B=@n4, Alpha=@n5), B=Constant3Vector() {constant=<rgba 0.0026 0.0045 0.0312 1>}, Alpha=@n14)
   B=Constant3Vector() {constant=<rgba 0.7639 0.7738 0.9167 1>}
   Alpha=Min(A=OneMinus(None=@n15), B=Multiply(A=@n16) {constB=2})
  )
  n18 = Constant() {r=0.25}
  MaterialFunctionCall!(
   Color=Power(Base=Desaturation(None=@n17, Fraction=@n18)) {constExponent=0.87}
   Color Brightness=Constant() {r=1}
   Floor Radius
   Floor Density
   Floor Metallic
   Floor Specular
   Floor Roughness
  ) {materialFunction=/Game/UI/Foundation/Environment/Shell/Materials/MaterialFunctions/MF_UI_MPItemShop_Setup}
  n19 = Constant2Vector() {g=1024, r=2048}
  MaterialFunctionCall(TextureSize=@n19) {materialFunction=&MF_UI_UVScaling}
  n20 = MaterialFunctionCall!(ScreenUV, TextureSize=@n19, ScalePivotPoint) {materialFunction=/FortUILibrary/MaterialFunctions/Basics/MF_UI_ScreenSpaceUVScaling}
  NamedRerouteDeclaration(None=@n20.Scaled to Fit) {name=UV Scaling_1, nodeColor=<rgba 0 0.0625 1 1>}
  NamedRerouteDeclaration(None=@n20.Scaled to Fill) {name=UVScaling Fill_1, nodeColor=<rgba 0.5469 1 0 1>}
  n21 = MaterialFunctionCall() {materialFunction=&MF_UI_ScreenSpaceUVScaling}
  n22 = TextureSample(UVs=Multiply(A=@n21.Scaled to Fit) {constB=1.5}) {constMipValue=1, samplerType=SAMPLERTYPE_LinearColor, texture=/Game/Packages/Fortress_Sky/SkyDome/Master/T_SkyDome_StarCloudMask}
  n23 = TextureSample(UVs=Panner(
    Coordinate=Multiply(A=@n21.Scaled to Fit) {constB=0.1}
    Speed=ScalarParameter() {defaultValue=0.025, group=Stars, parameterName=StarFlickerSpeed}
   )) {samplerType=SAMPLERTYPE_LinearColor, texture=&T_SkyDome_StarCloudMask}
  n24 = ShadingPathSwitch(Default=Multiply(A=ComponentMask(None=@n22.R) {r=true}, B=Multiply(A=@n23.G) {constB=0.3}))
  n25 = ScalarParameter() {defaultValue=2, group=Stars, parameterName=StarPower}
  NamedRerouteDeclaration(None=Multiply(A=Power(Base=@n24, Exp=@n25), B=Constant() {r=100})) {name=Stars_1, nodeColor=<rgba 0.8414 0 1 1>}
 ]
}
